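# Material function for the male medium player jumpsuit that adds optional color tinting and a spray-paint layer to incoming material attributes.
# Tint recolors balaclava, jumpsuit and gear with color parameters, using desaturated base color, and hard-light blends skin color.
# Spray composites SprayMap (UV0 or UV1) through its alpha and region masks, including an intensity-scaled zipper, blending toward non-metallic, rougher, flatter-normal attributes.
# Static switches compile each stage in or out.
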MaterialFunction /Game/Characters/Player/Male/Medium/Bodies/M_MED_Jumpsuit/Materials/MF_JumpSuit_Spray_Mutable
graph {nodes [
  Constant3Vector() {constant=<rgba 0.0833 0.042 0.0213 1>}
  Constant3Vector() {constant=<rgba 0.6458 0.5124 0.2456 1>}
  Constant3Vector() {constant=<rgba 0.5 0.3529 0.2057 1>}
  Constant3Vector() {constant=<rgba 0.2708 0.1422 0.0677 1>}
  Constant3Vector() {constant=<rgba 0.9216 0.6105 0.4072 1>}
  Constant3Vector() {constant=<rgba 0.0908 0.0409 0.0222 1>}
  Constant3Vector() {constant=<rgba 0.7529 0.4852 0.2423 1>}
  Constant3Vector() {constant=<rgba 0.5972 0.3419 0.1878 1>}
  Constant3Vector() {constant=<rgba 0.3231 0.1274 0.0545 1>}
  Constant3Vector() {constant=<rgba 0.6458 0.5108 0.3566 1>}
  Constant3Vector() {constant=<rgba 0.0625 0.0315 0.016 1>}
  Constant3Vector() {constant=<rgba 0.5885 0.4546 0.2391 1>}
  Constant3Vector() {constant=<rgba 0.4323 0.3051 0.1779 1>}
  Constant3Vector() {constant=<rgba 0.1979 0.1039 0.0495 1>}
  Constant3Vector() {constant=<rgba 0.6771 0.5356 0.3738 1>}
  n1 = NamedRerouteUsage()
  Desaturation(None=@n1)
  n2 = StaticSwitchParameter(
   True=SetMaterialAttributes!(MaterialAttributes=NamedRerouteUsage(), BaseColor=NamedRerouteUsage()) {attributeSetTypes [#1]}
   False=NamedRerouteUsage()
  ) {group=1-Jumpsuit, parameterName=UseJumpsuit_Tint, sortPriority=0}
  n3 = GetMaterialAttributes(None=@n2) {attributeGetTypes [#1, #2]}
  n4 = SetMaterialAttributes!(
   MaterialAttributes=@n3.MaterialAttributes
   BaseColor=NamedRerouteUsage()
   Metallic=Constant()
   Roughness=Constant() {r=0.8}
   Normal=MaterialFunctionCall(Normal=@n3.Normal, Flatness=Constant() {r=0.666}) {materialFunction=&FlattenNormal}.Result
  ) {attributeSetTypes [
    #1
    #3
    #4
    #2
   ]}
  n5 = NamedRerouteUsage()
  FunctionOutput(None=StaticSwitchParameter(True=BlendMaterialAttributes(A=@n3.MaterialAttributes, B=@n4, Alpha=@n5), False=@n2) {group=1-Jumpsuit, parameterName=UseJumpsuit_Spray, sortPriority=0})
  n6 = NamedRerouteDeclaration(None=@n1) {name=DesatBaseColor, nodeColor=<rgba 1 1 1 1>}
  n7 = VectorParameter() {group=1-Jumpsuit_Tint, parameterName=Color_Balaclava, defaultValue=<rgba 0.1268 0.1338 0.1458 1>}
  n8 = LinearInterpolate(
   A=LinearInterpolate(A=NamedRerouteUsage(), B=Multiply(A=@n6, B=@n7.RGB), Alpha=NamedRerouteUsage())
   B=Multiply(
    A=NamedRerouteUsage()
    B=VectorParameter() {group=1-Jumpsuit_Tint, parameterName=Color_Jumpsuit, defaultValue=<rgba 0.108 0.1875 0.0586 1>}.RGB
   )
   Alpha=NamedRerouteUsage()
  )
  n9 = Multiply(A=NamedRerouteUsage(), B=VectorParameter() {group=1-Jumpsuit_Tint, parameterName=Color_Gear, defaultValue=<rgba 0.7708 0.0723 0.0723 1>}.RGB)
  n10 = NamedRerouteUsage()
  n11 = NamedRerouteUsage()
  n12 = VectorParameter() {group=1-Jumpsuit_Tint, parameterName=Color_Skin, defaultValue=<rgba 0.4323 0.3051 0.1779 1>}
  NamedRerouteDeclaration(None=LinearInterpolate(
    A=LinearInterpolate(A=@n8, B=@n9, Alpha=@n10)
    B=MaterialFunctionCall!(Base=@n11, Blend=@n12.RGB) {materialFunction=/Engine/Functions/Engine_MaterialFunctions03/Blends/Blend_HardLight}.Result
    Alpha=NamedRerouteUsage()
   )) {name=TintedBaseColor, nodeColor=<rgba 1 0.2911 0.0904 1>}
  n13 = NamedRerouteUsage()
  n14 = TextureSampleParameter2D(UVs=StaticSwitchParameter(True=TextureCoordinate() {coordinateIndex=1}, False=TextureCoordinate()) {group=1-Jumpsuit_Spray, parameterName=UseSecondUVChannelForSpray, sortPriority=0}) {
   group=1-Jumpsuit_Spray
   parameterName=SprayMap
   sortPriority=2
   texture=/Game/Global/Textures/Default/Blanks/T_Blank_White_ZeroAlpha_sRGBColor
  }
  n15 = NamedRerouteUsage()
  n16 = NamedRerouteUsage()
  n17 = NamedRerouteUsage()
  n18 = ScalarParameter() {defaultValue=0.6, group=1-Jumpsuit_Spray, parameterName=Zipper_SprayIntensity, sliderMax=1, sortPriority=10}
  n19 = NamedRerouteUsage()
  n20 = Multiply(
   A=Max(A=Max(A=@n15, B=@n16), B=Multiply(A=@n17, B=@n18))
   B=MaterialFunctionCall(In=Multiply(A=@n19, B=@n19), Contrast=Constant() {r=0.5}) {materialFunction=&CheapContrast}.Result
  )
  n21 = OneMinus(None=NamedRerouteDeclaration(None=Multiply(A=@n14.A, B=@n20)) {name=SprayMask, nodeColor=<rgba 1 0.0984 0.845 1>})
  MaterialExpressionRemap_0 = MaterialXRemap!(None=@n14.RGB, InputLow, InputHigh, TargetLow, TargetHigh) {targetHighDefault=0.9, targetLowDefault=0.01}
  n22 = Multiply(A=@MaterialExpressionRemap_0, B=@n14.A)
  NamedRerouteDeclaration(None=Add(A=Multiply(A=@n13, B=@n21), B=Multiply(A=@n22, B=@n20))) {name=BaseColor_wSpray, nodeColor=<rgba 1 0.6563 0 1>}
  NamedRerouteDeclaration(None=FunctionInput() {inputName=AO, inputType=FunctionInput_Scalar}) {name=AO, nodeColor=<rgba 1 1 0 1>}
  NamedRerouteDeclaration(None=@n3.BaseColor) {name=BaseColor_afterTint, nodeColor=<rgba 0 0.8594 1 1>}
  n23 = TextureSampleParameter2D() {
   group=1-JumpSuit_Spray
   parameterName=SkinFX_Mask
   samplerType=SAMPLERTYPE_Masks
   sortPriority=2
   texture=&T_BlackMask
  }
  n24 = VectorParameter() {group=1-Jumpsuit_Spray, parameterName=FXMask_Zipper}
  NamedRerouteDeclaration(None=DotProduct(A=@n23.RGBA, B=AppendVector(A=@n24.RGB, B=@n24.RGB))) {name=FXMask_Zipper, nodeColor=<rgba 1 0 0.8142 1>}
  NamedRerouteDeclaration(None=FunctionInput() {inputName=Skin, inputType=FunctionInput_Scalar, sortPriority=2}) {name=SkinMask, nodeColor=<rgba 0 0.0424 1 1>}
  NamedRerouteDeclaration(None=FunctionInput() {inputName=GMap, inputType=FunctionInput_Scalar, sortPriority=1}) {name=GMap, desc=RimMask, nodeColor=<rgba 0 1 0.1687 1>}
  n25 = ChannelMaskParameter(None=@n23.RGBA) {group=1-Jumpsuit_Spray, maskChannel=Green, parameterName=FXMask_Jumpsuit, sortPriority=4, defaultValue=<rgba 0 1 0 0>}
  NamedRerouteDeclaration(None=@n25.R) {name=FXMask_Jumpsuit, nodeColor=<rgba 0 1 0.0108 1>}
  NamedRerouteDeclaration(None=ChannelMaskParameter(None=@n23.RGBA) {group=1-Jumpsuit_Spray, parameterName=FXMask_Balaclava, sortPriority=3}.R) {name=FXMask_Balaclava, nodeColor=<rgba 1 0.0075 0 1>}
  n26 = ChannelMaskParameter(None=@n23.RGBA) {group=1-Jumpsuit_Spray, maskChannel=Blue, parameterName=FXMask_Gear, sortPriority=5, defaultValue=<rgba 0 0 1 0>}
  NamedRerouteDeclaration(None=@n26.R) {name=FXMask_Gear, nodeColor=<rgba 0 0.0385 1 1>}
  n27 = FunctionInput() {inputName=MA, inputType=FunctionInput_MaterialAttributes, sortPriority=5}
  NamedRerouteDeclaration(None=GetMaterialAttributes(None=NamedRerouteDeclaration(None=@n27) {name=MA, nodeColor=<rgba 0 0.6484 1 1>}) {attributeGetTypes [#1]}.BaseColor) {name=BaseColor_In, nodeColor=<rgba 1 0 0.6797 1>}
 ]}
