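# Parent character material for the mutable medium male player jumpsuit, using material attributes with Default Lit and Subsurface Profile shading.
# The shared character PreFX function feeds a jumpsuit spray function, whose result passes through cloth-sim world position offset with a constant SkinFX mask.
# That result runs through the character PostFX function, then a cinematic PostFX function.
# Usage is set explicitly for skeletal, cloth, morph, mesh-deformer and Nanite meshes.
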Material /Game/Characters/Player/Male/Medium/Bodies/M_MED_Jumpsuit/Materials/M_FN_Character_Mutable_Parent
graph {
 properties {
  bAutomaticallySetUsageInEditor=false
  bUseMaterialAttributes=true
  bUsedWithClothing=true
  bUsedWithMeshDeformer=true
  bUsedWithMorphTargets=true
  bUsedWithNanite=true
  bUsedWithSkeletalMesh=true
  pixelDepthOffsetMode=PDOM_Legacy
  shadingModel=MSM_FromMaterialExpression
  subsurfaceProfile=&SS_HeroSkin_02
  usedShadingModels=Default Lit | Subsurface Profile
 }
 nodes [
  n1 = MaterialFunctionCall() {materialFunction=&BaseFN_CharacterFunction_PreFX}
  n2 = MaterialFunctionCall(
   AO=@n1.AO
   Distance Blend=@n1.AO
   Fresnel=@n1.AO
   Gmap=@n1.AO
   Pre Skinned Local Position=@n1.AO
   Material=MaterialFunctionCall!(
    Material=MaterialFunctionCall!(AO=@n1.AO, GMap=@n1.AO, Skin=@n1.AO, MA=@n1.AO) {materialFunction=&/MF_JumpSuit_Spray_Mutable}.Result
    SkinFX_Mask=Constant() {r=1}
   ) {materialFunction=/Game/Global/MaterialFunctions/MF_ClothSimWPO_v2}.Result
  ) {materialFunction=&BaseFN_CharacterFunction_PostFX}
  MaterialFunctionCall(In=@n2.Result) {materialFunction=&CINE_MF_CharPostFX}
 ]
}
