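# Control Rig for the male medium Jumpsuit body that procedurally nudges the left collar bone dyn_collar_l as the head moves.
# A debug-drawn spherical pose reader on head, parented to neck_01, outputs the weight for a local X offset on that bone.
# The reusable "Set Transforms for Bone" function adds translate, Euler rotate and scale deltas to the current local transform.
# Debug-axes and Modify Transforms nodes for collar_l sit unconnected to execution, and the right collar is not driven.
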ControlRigBlueprint /Game/Characters/Player/Male/Medium/Bodies/M_MED_Jumpsuit/Meshes/M_MED_Jumpsuit_CtrlRig
graphs [
 ContainedGraph {nodes [
   n1 = RigVMFunctionEntryNode()
   RigUnit_GetTransform = &RigUnit_GetTransform(Space=LocalSpace, bInitial=false, Item=@n1.Joint)
   n2 = &RigVMFunction_MathQuaternionToEuler(RotationOrder=ZYX, Value=@RigUnit_GetTransform.Transform.Rotation)
   > RigUnit_SetTransform = &RigUnit_SetTransform(
    Space=LocalSpace
    Value=(Rotation=(X=0.0132,Y=0.4099,Z=-0.0222,W=0.9118),Translation=(X=23.6045,Y=-6.37,Z=140.1115),Scale3D=(X=1,Y=1,Z=1))
    bInitial=false
    Item=@n1.Joint
    Weight=@n1.Weight
    Value.Translation.Z=&RigVMFunction_MathFloatAdd(B=@n1.Translate Z, A=@RigUnit_GetTransform.Transform.Translation.Z) {template="Add::Execute(in A,in B,out Result)"}.Result
    Value.Scale3D.Z=&RigVMFunction_MathFloatAdd(B=@n1.Scale Z, A=@RigUnit_GetTransform.Transform.Scale3D.Z) {template="Add::Execute(in A,in B,out Result)"}.Result
    Value.Translation.Y=&RigVMFunction_MathFloatAdd(B=@n1.Translate Y, A=@RigUnit_GetTransform.Transform.Translation.Y) {template="Add::Execute(in A,in B,out Result)"}.Result
    Value.Scale3D.Y=&RigVMFunction_MathFloatAdd(B=@n1.Scale Y, A=@RigUnit_GetTransform.Transform.Scale3D.Y) {template="Add::Execute(in A,in B,out Result)"}.Result
    Value.Translation.X=&RigVMFunction_MathFloatAdd(B=@n1.Translate X, A=@RigUnit_GetTransform.Transform.Translation.X) {template="Add::Execute(in A,in B,out Result)"}.Result
    Value.Scale3D.X=&RigVMFunction_MathFloatAdd(B=@n1.Scale X, A=@RigUnit_GetTransform.Transform.Scale3D.X) {template="Add::Execute(in A,in B,out Result)"}.Result
    Value.Rotation=&RigVMFunction_MathQuaternionFromEuler(
     Euler=(X=0,Y=0,Z=0)
     Euler.X=&RigVMFunction_MathFloatAdd(B=@n1.Rotate X, A=@n2.Result.X) {template="Add::Execute(in A,in B,out Result)"}.Result
     Euler.Y=&RigVMFunction_MathFloatAdd(B=@n1.Rotate Y, A=@n2.Result.Y) {template="Add::Execute(in A,in B,out Result)"}.Result
     Euler.Z=&RigVMFunction_MathFloatAdd(B=@n1.Rotate Z, A=@n2.Result.Z) {template="Add::Execute(in A,in B,out Result)"}.Result
    ).Result
   )
   > RigVMFunctionReturnNode()
  ]}
 RigVMFunctionLibrary {nodes [Set Transforms for Bone = RigVMCollapseNode(
    ExecuteContext=()
    Joint=()
    Rotate X=0
    Rotate Y=0
    Rotate Z=0
    Scale X=0
    Scale Y=0
    Scale Z=0
    Translate X=0
    Translate Y=0
    Translate Z=0
    Weight=1
   )]}
 RigVMModel {nodes [
   RigUnit_BeginExecution = &RigUnit_BeginExecution()
   n1 = &RigUnit_SphericalPoseReader(
    ActiveRegionScaleFactors=(PositiveWidth=1,NegativeWidth=1,PositiveHeight=1,NegativeHeight=1)
    ActiveRegionSize=0.2
    Debug=(bDrawDebug=True,bDrawLocalAxes=true,DebugScale=25,DebugSegments=20,DebugThickness=0.25)
    DriverAxis=(X=1,Y=0,Z=0)
    DriverItem=(Type=Bone,Name="head")
    FalloffRegionScaleFactors=(PositiveWidth=1,NegativeWidth=1,PositiveHeight=1,NegativeHeight=1)
    FalloffSize=0.3
    FlipWidthScaling=true
    OptionalParentItem=(Type=Bone,Name="neck_01")
    RotationOffset=(X=120,Y=0,Z=-30)
    ExecutePin=@RigUnit_BeginExecution.ExecutePin
   )
   > Set Transforms for Bone_0 = RigVMFunctionReferenceNode(
    Joint=(Type=Bone,Name="dyn_collar_l")
    Rotate X=0
    Rotate Y=0
    Rotate Z=0
    Scale X=0
    Scale Y=0
    Scale Z=0
    Translate X=-5
    Translate Y=0
    Translate Z=0
    Weight=@n1.OutputParam
   )
   /Script/ControlRig.RigUnit_DebugTransformMutableItemSpace(
    Color=(R=1,G=0,B=0,A=1)
    DebugDrawSettings=(DepthPriority=SDPG_Foreground,Lifetime=-1)
    ExecutePin=()
    Mode=Axes
    Scale=5
    Space=(Type=Bone)
    Thickness=0.1
    WorldOffset=(Rotation=(X=0,Y=0,Z=0,W=1),Translation=(X=0,Y=0,Z=0),Scale3D=(X=1,Y=1,Z=1))
    bEnabled=True
    Transform=&RigUnit_GetTransform(Item=(Type=Bone,Name="collar_l")).Transform
   ) {is=struct RigVMUnitNode}
   &RigUnit_ModifyTransforms(
    ExecutePin=()
    ItemToModify=((Item=(Type=Bone,Name="collar_l"),Transform=(Rotation=(X=0,Y=0,Z=0,W=1),Translation=(X=100,Y=100,Z=0),Scale3D=(X=1,Y=1,Z=1))))
    Weight=1
   )
  ]}
]
hierarchy [
 root {loc=<0 0 0>, rot=<0 0 0 1>}
 pelvis {parent=root, loc=<0 -0.7071 93.1191>, rot=<0.0127 0.707 -0.0127 -0.707>}
 spine_01 {parent=pelvis, loc=<6.7494 -0.6388 0>, rot=<0 0 0.034 -0.9994>}
 spine_02 {parent=spine_01, loc=<7.2518 -0.035 0>, rot=<0 0 0.0238 -0.9997>}
 spine_03 {parent=spine_02, loc=<6.0104 0 0>, rot=<0 0 0.0149 -0.9999>}
 spine_04 {parent=spine_03, loc=<8.1024 0 0>, rot=<0 0 -0.0132 -0.9999>}
 spine_05 {parent=spine_04, loc=<10.0146 0 0>, rot=<0 0 -0.0015 -1>}
 neck_01 {parent=spine_05, loc=<11.3138 0.1529 0>, rot=<0 0 -0.2003 -0.9797>}
 neck_02 {parent=neck_01, loc=<4.5627 0 0>, rot=<0 0 0.0504 -0.9987>}
 head {parent=neck_02, loc=<5.3914 0 0>, rot=<0 0 0.115 -0.9934>}
 faceAttach {parent=head, loc=<-2.277 0.4707 -0.0117>, rot=<0.0082 -0.7071 -0.0082 -0.7071>, scale=<0.89 0.89 0.89>}
 C_jaw {parent=faceAttach, loc=<-0.0472 3.1937 0.7586>, rot=<0 0 0 1>}
 collar_l {parent=neck_02, loc=<-1.685 3.1434 -5.8839>, rot=<-0.5988 -0.7504 -0.1703 0.2223>}
 collar_r {parent=neck_02, loc=<-1.6852 3.1434 5.8839>, rot=<0.7504 -0.5988 0.2223 0.1703>}
 dyn_collar_l {parent=neck_02, loc=<-1.685 3.1434 -5.8839>, rot=<-0.5988 -0.7504 -0.1703 0.2223>}
 dyn_collar_r {parent=neck_02, loc=<-1.6852 3.1434 5.8839>, rot=<0.7504 -0.5988 0.2223 0.1703>}
 dyn_neck {parent=neck_01, loc=<4.5627 0 0>, rot=<0 0 0.0504 -0.9987>}
 clavicle_l {parent=spine_05, loc=<4.3575 -0.1466 -2.3223>, rot=<0.0465 -0.7289 -0.009 -0.6829>}
 upperarm_l {parent=clavicle_l, loc=<15.5111 0 0>, rot=<0.0226 -0.3234 -0.0088 -0.9459>}
 lowerarm_l {parent=upperarm_l, loc=<26.6514 0 0>, rot=<0.0391 -0.0099 -0.2453 -0.9686>}
 hand_l {parent=lowerarm_l, loc=<22.3491 0 0>, rot=<-0.7081 -0.0507 0.0975 -0.6975>}
 thumb_01_l {parent=hand_l, loc=<1.3118 -0.3191 -2.459>, rot=<0.6546 -0.4042 0.0664 -0.6354>}
 thumb_02_l {parent=thumb_01_l, loc=<5.9336 -0.0448 0.0401>, rot=<-0.0037 0.0213 0.298 -0.9543>}
 thumb_03_l {parent=thumb_02_l, loc=<2.2223 0.3682 -0.0907>, rot=<0.0118 -0.013 0.0066 -0.9998>}
 index_metacarpal_l {parent=hand_l, loc=<1.6535 0.0183 -2.96>, rot=<0.0158 -0.0124 -0.0173 -0.9996>}
 index_01_l {parent=index_metacarpal_l, loc=<8.3008 0.0357 -0.0524>, rot=<-0.0044 0.0485 0.0924 -0.9945>}
 index_02_l {parent=index_01_l, loc=<3.3587 -0.3004 0.0452>, rot=<-0.002 0.0027 0.3225 -0.9466>}
 index_03_l {parent=index_02_l, loc=<2.5578 -0.123 -0.0252>, rot=<-0.002 0.0067 0.2558 -0.9667>}
 middle_metacarpal_l {parent=hand_l, loc=<2.1856 -0.1725 -0.5984>, rot=<0.006 0.0168 -0.0329 -0.9993>}
 middle_01_l {parent=middle_metacarpal_l, loc=<7.3275 -0.2447 0.0129>, rot=<-0.0208 0.0042 0.0858 -0.9961>}
 middle_02_l {parent=middle_01_l, loc=<4.0123 0.1401 0.1175>, rot=<-0.0218 0.0258 0.3512 -0.9357>}
 middle_03_l {parent=middle_02_l, loc=<2.6077 0.0534 0.1373>, rot=<0.0082 0.0188 0.1708 -0.9851>}
 ring_metacarpal_l {parent=hand_l, loc=<2.2074 -0.3625 1.3638>, rot=<-0.0284 0.0722 -0.0458 -0.9959>}
 ring_01_l {parent=ring_metacarpal_l, loc=<7.0295 -0.6197 -0.1446>, rot=<-0.0489 -0.0252 0.1892 -0.9804>}
 ring_02_l {parent=ring_01_l, loc=<3.6174 0.4647 -0.1027>, rot=<0 0 0.2806 -0.9598>}
 ring_03_l {parent=ring_02_l, loc=<2.4152 0.114 0.0199>, rot=<0 0 0.1189 -0.9929>}
 pinky_metacarpal_l {parent=hand_l, loc=<1.7616 -0.9753 3.3814>, rot=<-0.0397 0.0814 -0.069 -0.9935>}
 pinky_01_l {parent=pinky_metacarpal_l, loc=<7.1709 -0.4831 -0.1094>, rot=<-0.0654 -0.0035 0.2242 -0.9723>}
 pinky_02_l {parent=pinky_01_l, loc=<2.8274 0.0893 -0.1043>, rot=<0.0037 0.0011 0.2854 -0.9584>}
 pinky_03_l {parent=pinky_02_l, loc=<2.2046 0.0307 -0.0078>, rot=<-0.0002 0.0098 0.1445 -0.9895>}
 weapon_l {parent=hand_l, loc=<-0.1974 -3.4111 0.3811>, rot=<0 0 0 -1>}
 lowerarm_twist_02_l {parent=lowerarm_l, loc=<9.6199 0 0>, rot=<0 0 0 -1>}
 lowerarm_twist_01_l {parent=lowerarm_l, loc=<17.8549 0 0>, rot=<0 0 0 -1>}
 upperarm_twist_01_l {parent=upperarm_l, loc=<0.0041 0 0>, rot=<0 0 0 -1>}
 upperarm_twist_02_l {parent=upperarm_l, loc=<13.7941 0 0>, rot=<0 0 0 -1>}
 dyn_l_sleeve_a {parent=upperarm_l, loc=<-0.0816 -5.2135 3.8129>, rot=<-0.3469 0.407 -0.1539 -0.8309>}
 dyn_l_sleeveCenter {parent=upperarm_l, loc=<0 0 0>, rot=<0 0 0 1>}
 clavicle_r {parent=spine_05, loc=<4.3575 -0.1466 2.3223>, rot=<0.7289 0.0465 -0.6829 0.009>}
 upperarm_r {parent=clavicle_r, loc=<-15.5111 0 0>, rot=<0.0226 -0.3234 -0.0088 -0.9459>}
 lowerarm_r {parent=upperarm_r, loc=<-26.6514 0 0>, rot=<0.0391 -0.0099 -0.2453 -0.9686>}
 hand_r {parent=lowerarm_r, loc=<-22.3491 0 0>, rot=<-0.7081 -0.0507 0.0975 -0.6975>}
 thumb_01_r {parent=hand_r, loc=<-1.3119 0.3192 2.459>, rot=<0.6546 -0.4042 0.0664 -0.6354>}
 thumb_02_r {parent=thumb_01_r, loc=<-5.9336 0.0448 -0.0401>, rot=<-0.0037 0.0213 0.298 -0.9543>}
 thumb_03_r {parent=thumb_02_r, loc=<-2.2223 -0.3682 0.0907>, rot=<0.0118 -0.013 0.0066 -0.9998>}
 index_metacarpal_r {parent=hand_r, loc=<-1.6534 -0.0181 2.9601>, rot=<0.0158 -0.0124 -0.0173 -0.9996>}
 index_01_r {parent=index_metacarpal_r, loc=<-8.3008 -0.0357 0.0524>, rot=<-0.0044 0.0485 0.0924 -0.9945>}
 index_02_r {parent=index_01_r, loc=<-3.3589 0.3003 -0.0453>, rot=<-0.002 0.0027 0.3225 -0.9466>}
 index_03_r {parent=index_02_r, loc=<-2.5578 0.123 0.0252>, rot=<-0.002 0.0067 0.2558 -0.9667>}
 middle_metacarpal_r {parent=hand_r, loc=<-2.1856 0.1726 0.5984>, rot=<0.006 0.0168 -0.0329 -0.9993>}
 middle_01_r {parent=middle_metacarpal_r, loc=<-7.3273 0.2447 -0.0129>, rot=<-0.0208 0.0042 0.0858 -0.9961>}
 middle_02_r {parent=middle_01_r, loc=<-4.0142 -0.1411 -0.07>, rot=<-0.0217 0.0259 0.3512 -0.9357>}
 middle_03_r {parent=middle_02_r, loc=<-2.6074 -0.0534 -0.1312>, rot=<0.0083 0.0188 0.1708 -0.9851>}
 ring_metacarpal_r {parent=hand_r, loc=<-2.2073 0.3627 -1.3638>, rot=<-0.0284 0.0722 -0.0458 -0.9959>}
 ring_01_r {parent=ring_metacarpal_r, loc=<-7.0297 0.6197 0.1446>, rot=<-0.0489 -0.0252 0.1892 -0.9804>}
 ring_02_r {parent=ring_01_r, loc=<-3.6171 -0.4648 0.1027>, rot=<0 0 0.2806 -0.9598>}
 ring_03_r {parent=ring_02_r, loc=<-2.4152 -0.114 -0.0199>, rot=<0 0 0.1189 -0.9929>}
 pinky_metacarpal_r {parent=hand_r, loc=<-1.7617 0.9753 -3.3814>, rot=<-0.0397 0.0814 -0.069 -0.9935>}
 pinky_01_r {parent=pinky_metacarpal_r, loc=<-7.1707 0.4832 0.1094>, rot=<-0.0654 -0.0035 0.2242 -0.9723>}
 pinky_02_r {parent=pinky_01_r, loc=<-2.8274 -0.0893 0.1043>, rot=<0.0037 0.0011 0.2854 -0.9584>}
 pinky_03_r {parent=pinky_02_r, loc=<-2.2045 -0.031 0.0078>, rot=<-0.0002 0.0098 0.1445 -0.9895>}
 lowerarm_twist_02_r {parent=lowerarm_r, loc=<-9.6199 0 0>, rot=<0 0 0 -1>}
 lowerarm_twist_01_r {parent=lowerarm_r, loc=<-17.8549 0 0>, rot=<0 0 0 -1>}
 upperarm_twist_01_r {parent=upperarm_r, loc=<-0.0041 0 0>, rot=<0 0 0 -1>}
 upperarm_twist_02_r {parent=upperarm_r, loc=<-13.7941 0 0>, rot=<0 0 0 -1>}
 dyn_r_sleeve {parent=upperarm_r, loc=<-4.6078 0 0>, rot=<0 0 0 1>}
 attach_backpack {parent=spine_05, loc=<-0.123 -13.4727 0>, rot=<0 -0.7071 0 -0.7071>}
 attach_cape {parent=spine_05, loc=<9.017 -11.714 0>, rot=<0.0284 -0.7065 -0.0284 -0.7065>}
 dyn_scarf_MSTR {parent=spine_05, loc=<6.9693 10.3416 -0.4524>, rot=<-0.1077 -0.7023 -0.6946 -0.1124>}
 dyn_scarf {parent=dyn_scarf_MSTR, loc=<0 0 0>, rot=<0 0 0 1>}
 dyn_hoodie_MSTR {parent=spine_05, loc=<13.286 -4.656 -0.3459>, rot=<0.2489 -0.6511 0.6719 -0.2502>}
 dyn_hoodie_01 {parent=dyn_hoodie_MSTR, loc=<0 0 0>, rot=<0 0 0 1>}
 dyn_hoodie_02 {parent=dyn_hoodie_01, loc=<5.1412 0 0>, rot=<0.0067 0.0618 0.0306 -0.9976>}
 dyn_skyDiveHoodFix {parent=spine_05, loc=<17.3062 -9.699 -0.5233>, rot=<0.1768 -0.6711 0.7064 -0.1391>}
 MSTR_chest_zipper {parent=spine_04, loc=<-1.983 12.5567 -5.7764>, rot=<0.0024 -0.7929 -0.6091 -0.0145>}
 chest_zipper {parent=MSTR_chest_zipper, loc=<0 0 0>, rot=<0 0 0 1>}
 dyn_chest_zipper {parent=MSTR_chest_zipper, loc=<0 0 0>, rot=<0 0 0 1>}
 thigh_l {parent=pelvis, loc=<-2.5693 -0.0109 -7.7501>, rot=<0.0012 0.0572 0.0155 -0.9982>}
 calf_l {parent=thigh_l, loc=<-39.0484 0 0>, rot=<0.0006 -0.0161 -0.0391 -0.9991>}
 foot_l {parent=calf_l, loc=<-42.561 0.1077 0.4292>, rot=<0.0418 -0.0455 0.0343 -0.9975>}
 ball_l {parent=foot_l, loc=<-6.1599 13.8264 0.6806>, rot=<0.0001 -0.0001 -0.7063 -0.7079>}
 calf_twist_02_l {parent=calf_l, loc=<-18.8064 -0.0967 0.0397>, rot=<0.0237 0.0095 -0.0035 -0.9997>}
 calf_twist_01_l {parent=calf_l, loc=<-35.813 -0.0967 0.0397>, rot=<0.0237 0.0095 -0.0035 -0.9997>}
 dyn_l_kneeFold {parent=calf_l, loc=<0 0 0>, rot=<0 0 0 1>}
 MSTR_front_hem_l {parent=calf_l, loc=<-32.0855 -1.7073 0.3935>, rot=<-0.1502 -0.6976 -0.6944 -0.0921>}
 dyn_front_hem_l {parent=MSTR_front_hem_l, loc=<0 0 0>, rot=<0 0 0 1>}
 MSTR_back_hem_l {parent=calf_l, loc=<-32.1692 -2.7147 0.3906>, rot=<0.0916 -0.7097 0.6977 -0.0346>}
 dyn_back_hem_l {parent=MSTR_back_hem_l, loc=<0 0 0>, rot=<0 0 0 1>}
 thigh_twist_01_l {parent=thigh_l, loc=<-21.6303 0 0>, rot=<0.0013 0 0 -1>}
 MSTR_pouchFlap_l {parent=thigh_twist_01_l, loc=<-4.9949 5.9312 -7.026>, rot=<0.0084 -0.9396 -0.3395 0.0423>}
 pouchFlap_0_l {parent=MSTR_pouchFlap_l, loc=<0 0 0>, rot=<0 0 0 1>}
 pouchFlap_1_l {parent=pouchFlap_0_l, loc=<5.1746 0 0>, rot=<0 -0.0172 0 -0.9999>}
 pouchFlap_2_l {parent=pouchFlap_1_l, loc=<6.098 0 0>, rot=<0 0 0 1>}
 dyn_pouchFlap_0_l {parent=MSTR_pouchFlap_l, loc=<0 0 0>, rot=<0 0 0 1>}
 dyn_pouchFlap_1_l {parent=dyn_pouchFlap_0_l, loc=<5.1746 0 0>, rot=<0 -0.0172 0 -0.9999>}
 dyn_pouchFlap_2_l {parent=dyn_pouchFlap_1_l, loc=<6.098 0 0>, rot=<0 0 0 1>}
 MSTR_pouch_zipper {parent=thigh_l, loc=<-13.4195 10.8235 -2.5773>, rot=<-0.0033 0.4278 0.8986 -0.0974>}
 pouch_zipper {parent=MSTR_pouch_zipper, loc=<0 0 0>, rot=<0 0 0 1>}
 dyn_pouch_zipper {parent=MSTR_pouch_zipper, loc=<0 0 0>, rot=<0 0 0 1>}
 MSTR_leg_l_zipper {parent=thigh_l, loc=<-16.7928 -2.7758 -8.3287>, rot=<0.0004 -0.9952 0.0886 -0.0417>}
 leg_l_zipper {parent=MSTR_leg_l_zipper, loc=<0 0 0>, rot=<0 0 0 1>}
 dyn_leg_l_zipper {parent=MSTR_leg_l_zipper, loc=<0 0 0>, rot=<0 0 0 1>}
 thigh_r {parent=pelvis, loc=<-2.5691 -0.0109 7.7501>, rot=<0.0572 -0.0012 0.9982 0.0155>}
 calf_r {parent=thigh_r, loc=<39.0487 0 0>, rot=<0.0006 -0.0161 -0.0391 -0.9991>}
 foot_r {parent=calf_r, loc=<42.5612 -0.1077 -0.4292>, rot=<0.0418 -0.0455 0.0343 -0.9975>}
 ball_r {parent=foot_r, loc=<6.1599 -13.8264 -0.6806>, rot=<0.0001 -0.0001 -0.7063 -0.7079>}
 calf_twist_02_r {parent=calf_r, loc=<18.8064 0.0967 -0.0397>, rot=<0.0237 0.0095 -0.0035 -0.9997>}
 calf_twist_01_r {parent=calf_r, loc=<35.8132 0.0967 -0.0397>, rot=<0.0237 0.0095 -0.0035 -0.9997>}
 dyn_r_kneeFold {parent=calf_r, loc=<0 0 0>, rot=<0 0 0 1>}
 MSTR_front_hem_r {parent=calf_r, loc=<32.0854 1.7073 -0.3935>, rot=<-0.1502 -0.6976 -0.6944 -0.0921>}
 dyn_front_hem_r {parent=MSTR_front_hem_r, loc=<0 0 0>, rot=<0 0 0 1>}
 MSTR_back_hem_r {parent=calf_r, loc=<32.1692 2.7147 -0.3906>, rot=<0.0916 -0.7097 0.6977 -0.0346>}
 dyn_back_hem_r {parent=MSTR_back_hem_r, loc=<0 0 0>, rot=<0 0 0 1>}
 thigh_twist_01_r {parent=thigh_r, loc=<21.6303 0 0>, rot=<0.0013 0 0 -1>}
 MSTR_pouchFlap_r {parent=thigh_twist_01_r, loc=<4.9951 -5.9311 7.026>, rot=<-0.3395 -0.0423 -0.0084 -0.9396>}
 pouchFlap_0_r {parent=MSTR_pouchFlap_r, loc=<0 0 0>, rot=<0 0 0 1>}
 pouchFlap_1_r {parent=pouchFlap_0_r, loc=<5.1747 -0.0001 -0.0001>, rot=<0 -0.0172 0 -0.9999>}
 pouchFlap_2_r {parent=pouchFlap_1_r, loc=<6.0979 0 0>, rot=<0 0 0 1>}
 dyn_pouchFlap_0_r {parent=MSTR_pouchFlap_r, loc=<0 0 0>, rot=<0 0 0 1>}
 dyn_pouchFlap_1_r {parent=dyn_pouchFlap_0_r, loc=<5.1747 -0.0001 -0.0001>, rot=<0 -0.0172 0 -0.9999>}
 dyn_pouchFlap_2_r {parent=dyn_pouchFlap_1_r, loc=<6.0979 0 0>, rot=<0 0 0 1>}
 MSTR_leg_r_zipper {parent=thigh_r, loc=<12.2469 9.753 -3.7908>, rot=<0.7379 -0.1796 -0.2123 -0.615>}
 leg_r_zipper {parent=MSTR_leg_r_zipper, loc=<0 0 0>, rot=<0 0 0 1>}
 dyn_leg_r_zipper {parent=MSTR_leg_r_zipper, loc=<0 0 0>, rot=<0 0 0 1>}
 MSTR_pouchFlap_hip {parent=pelvis, loc=<4.3856 2.2404 16.0485>, rot=<0.2135 0.1121 0.9705 0.0056>}
 pouchFlap_hip_0 {parent=MSTR_pouchFlap_hip, loc=<0 0 0>, rot=<0 0 0 1>}
 pouchFlap_hip_1 {parent=pouchFlap_hip_0, loc=<4.2759 0 0>, rot=<0 0 0 1>}
 pouchFlap_hip_2 {parent=pouchFlap_hip_1, loc=<3.0993 0.0009 0.1433>, rot=<-0.0138 -0.2233 0.0032 -0.9746>}
 dyn_pouchFlap_hip_0 {parent=MSTR_pouchFlap_hip, loc=<0 0 0>, rot=<0 0 0 1>}
 dyn_pouchFlap_hip_1 {parent=dyn_pouchFlap_hip_0, loc=<4.2759 0 0>, rot=<0 0 0 1>}
 dyn_pouchFlap_hip_2 {parent=dyn_pouchFlap_hip_1, loc=<3.0993 0.0009 0.1433>, rot=<-0.0138 -0.2233 0.0032 -0.9746>}
 ik_hand_root {parent=root, loc=<0 0 0>, rot=<0 0 0 1>}
 ik_hand_gun {parent=ik_hand_root, loc=<-51.9729 4.0884 102.6935>, rot=<0.6084 0.1793 0.4613 -0.6205>}
 ik_hand_l {parent=ik_hand_gun, loc=<53.0286 82.1782 35.2051>, rot=<-0.5102 -0.7906 -0.3387 0>}
]
