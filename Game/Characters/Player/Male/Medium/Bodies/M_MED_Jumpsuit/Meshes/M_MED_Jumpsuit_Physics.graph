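# Physics asset for the male medium player jumpsuit, providing bodies for its dangling pouch flaps and chest zipper alongside the lower body.
# Capsules on pelvis, thigh twist bones and calves form the body, with thighs and calves each constrained directly to the pelvis.
# Each pouch flap is a master-plus-three-bone chain hung from its thigh, using narrow swing and very tight twist limits.
# The two-body zipper chain attaches to the pelvis with wider swing.
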PhysicsAsset /Game/Characters/Player/Male/Medium/Bodies/M_MED_Jumpsuit/Meshes/M_MED_Jumpsuit_Physics
bodies {
 MSTR_chest_zipper [{shapeType=Capsule, length=0.38, radius=0.63, shapeName=MSTR_chest_zipper_capsule, rotation=<rot 0 -90 0>}]
 MSTR_pouchFlap_l [{shapeType=Capsule, length=4, radius=1.5, shapeName=MSTR_pouchFlap_l_capsule, center=<2.3597 0 0.0832>, rotation=<rot -87.98 0.0002 -0.0001>}]
 MSTR_pouchFlap_r [{shapeType=Capsule, length=4, radius=1.5, shapeName=MSTR_pouchFlap_r_capsule, center=<2.3597 0 0.0832>, rotation=<rot -87.98 -0.0002 0.0001>}]
 calf_l [{shapeType=Capsule, length=24.4015, radius=9.3561, center=<-22.3415 -1.043 -0.3393>, rotation=<rot 0 86.6621 -89.9319>}]
 calf_r [{shapeType=Capsule, length=24.422, radius=9.6165, center=<22.3997 0.8927 0.0158>, rotation=<rot 0 87.2979 -89.2167>}]
 dyn_chest_zipper [{shapeType=Capsule, length=2.875, radius=1.425, shapeName=dyn_chest_zipper_capsule, center=<2.9087 0.0286 -0.0249>, rotation=<rot 0 -90 -90>}]
 dyn_pouchFlap_0_l [{
   shapeType=Capsule
   length=6.5
   radius=3
   shapeName=dyn_pouchFlap_0_l_capsule
   center=<3.6171 0.0664 -0.039>
   rotation=<rot -87.9797 0.0012 -91.6572>
  }]
 dyn_pouchFlap_0_r [{
   shapeType=Capsule
   length=4
   radius=2
   shapeName=dyn_pouchFlap_0_r_capsule
   center=<4.8254 0 0.1702>
   rotation=<rot -87.9799 -0.0005 93.0431>
  }]
 dyn_pouchFlap_1_l [{
   shapeType=Capsule
   length=6.5
   radius=3
   shapeName=dyn_pouchFlap_1_l_capsule
   center=<3.6171 0.0664 -0.039>
   rotation=<rot -87.9797 0.0008 -91.6567>
  }]
 dyn_pouchFlap_1_r [{
   shapeType=Capsule
   length=4
   radius=2
   shapeName=dyn_pouchFlap_1_r_capsule
   center=<4.8254 0 0.1702>
   rotation=<rot -87.9799 -0.0003 93.0429>
  }]
 dyn_pouchFlap_2_l [{
   shapeType=Capsule
   length=6.5
   radius=3
   shapeName=dyn_pouchFlap_2_l_capsule
   center=<2.1746 0.1081 -0.039>
   rotation=<rot -87.9797 0.0007 -91.6567>
  }]
 dyn_pouchFlap_2_r [{
   shapeType=Capsule
   length=3.5
   radius=2.5
   shapeName=dyn_pouchFlap_2_r_capsule
   center=<3.3451 -0.0787 0.118>
   rotation=<rot -87.9799 -0.0003 93.0429>
  }]
 pelvis [{shapeType=Capsule, length=15, radius=10}]
 thigh_twist_01_l [{shapeType=Capsule, length=23.2505, radius=9.253, rotation=<rot -89.683 -179.9991 179.9989>}]
 thigh_twist_01_r [{shapeType=Capsule, length=23.2505, radius=9.253, rotation=<rot -89.683 -179.9991 179.9989>}]
}
constraints [
 {bone1Name=calf_l, bone2Name=pelvis, swing1LimitDegrees=45, swing2LimitDegrees=45, twistLimitDegrees=45}
 {bone1Name=thigh_twist_01_l, bone2Name=pelvis, swing1LimitDegrees=45, swing2LimitDegrees=45, twistLimitDegrees=45}
 {bone1Name=calf_r, bone2Name=pelvis, swing1LimitDegrees=45, swing2LimitDegrees=45, twistLimitDegrees=45}
 {bone1Name=thigh_twist_01_r, bone2Name=pelvis, swing1LimitDegrees=45, swing2LimitDegrees=45, twistLimitDegrees=45}
 {bone1Name=dyn_pouchFlap_0_l, bone2Name=MSTR_pouchFlap_l, swing1LimitDegrees=20, swing2LimitDegrees=20, twistLimitDegrees=2}
 {bone1Name=dyn_pouchFlap_1_l, bone2Name=dyn_pouchFlap_0_l, swing1LimitDegrees=20, swing2LimitDegrees=20, twistLimitDegrees=2}
 {bone1Name=dyn_pouchFlap_2_l, bone2Name=dyn_pouchFlap_1_l, swing1LimitDegrees=20, swing2LimitDegrees=20, twistLimitDegrees=2}
 {bone1Name=MSTR_pouchFlap_l, bone2Name=thigh_twist_01_l, swing1LimitDegrees=20, swing2LimitDegrees=20, twistLimitDegrees=2}
 {bone1Name=MSTR_pouchFlap_r, bone2Name=thigh_twist_01_r, swing1LimitDegrees=20, swing2LimitDegrees=20, twistLimitDegrees=2}
 {bone1Name=dyn_pouchFlap_0_r, bone2Name=MSTR_pouchFlap_r, swing1LimitDegrees=20, swing2LimitDegrees=20, twistLimitDegrees=2}
 {bone1Name=dyn_pouchFlap_1_r, bone2Name=dyn_pouchFlap_0_r, swing1LimitDegrees=20, swing2LimitDegrees=20, twistLimitDegrees=2}
 {bone1Name=dyn_pouchFlap_2_r, bone2Name=dyn_pouchFlap_1_r, swing1LimitDegrees=20, swing2LimitDegrees=20, twistLimitDegrees=2}
 {bone1Name=MSTR_chest_zipper, bone2Name=pelvis, swing1LimitDegrees=45, swing2LimitDegrees=45, twistLimitDegrees=2}
 {bone1Name=dyn_chest_zipper, bone2Name=MSTR_chest_zipper, swing1LimitDegrees=45, swing2LimitDegrees=45, twistLimitDegrees=2}
]
