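# Post-process style AnimBP for the male medium player jumpsuit, adding secondary motion to hems, pouch flaps and zippers.
# It copies the parent mesh pose, adds RigidBody simulation only in the front end, then runs two Control Rigs.
# Constraints and AnimDynamics feed a cached pose, which branches into two hem-spring variants.
# The event graph drives material WPO cloth from leg bones, skydive-float wind and a crouch-based constraint alpha.
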AnimBlueprint /Game/Characters/Player/Male/Medium/Bodies/M_MED_Jumpsuit/Meshes/M_MED_Jumpsuit_AnimBP
target_skeleton=&/M_MED_Jumpsuit_Skeleton
graphs [
 AnimGraph {nodes [
   n1 = Anim Dynamics - Bone: dyn_back_hem_r(Alpha=1, ExternalForce="0,0,0") {
    angularDampingOverride=0.9999
    angularSpringConstant=75
    bAngularSpring=true
    bGravityOverrideInSimSpace=true
    bOverrideAngularDamping=true
    simulationSpace=RootRelative
    alphaBoolBlend {blendOption=Linear}
    boundBone {boneName=dyn_back_hem_r}
    componentLinearAccScale=<1 1 1>
    componentLinearVelScale=<1 1 1>
    physicsBodyDefinitions [{
      sphereCollisionRadius=0
      boundBone {boneName=dyn_back_hem_r}
      boxExtents=<10 20 20>
      constraintSetup {angularLimitsMax=<1 -10 5>, angularLimitsMin=<-1 -30 -5>, angularTarget=<0 0 0>}
      localJointOffset=<1 0 -0.2>
     }]
   }
   n2 = Anim Dynamics - Bone: dyn_front_hem_r(ComponentPose=@n1, Alpha=1, ExternalForce="0,0,0") {
    angularDampingOverride=0.9999
    angularSpringConstant=75
    bAngularSpring=true
    bGravityOverrideInSimSpace=true
    bOverrideAngularDamping=true
    simulationSpace=RootRelative
    alphaBoolBlend {blendOption=Linear}
    boundBone {boneName=dyn_front_hem_r}
    componentLinearAccScale=<1 1 1>
    componentLinearVelScale=<1 1 1>
    physicsBodyDefinitions [{
      sphereCollisionRadius=0
      boundBone {boneName=dyn_front_hem_r}
      boxExtents=<10 20 20>
      constraintSetup {angularLimitsMax=<1 -10 10>, angularLimitsMin=<-1 -30 0>, angularTarget=<0 0 0>}
      localJointOffset=<1 0 -0.2>
     }]
   }
   n3 = Anim Dynamics - Bone: dyn_back_hem_l(ComponentPose=@n2, Alpha=1, ExternalForce="0,0,0") {
    angularDampingOverride=0.9999
    angularSpringConstant=75
    bAngularSpring=true
    bGravityOverrideInSimSpace=true
    bOverrideAngularDamping=true
    simulationSpace=RootRelative
    alphaBoolBlend {blendOption=Linear}
    boundBone {boneName=dyn_back_hem_l}
    componentLinearAccScale=<1 1 1>
    componentLinearVelScale=<1 1 1>
    physicsBodyDefinitions [{
      sphereCollisionRadius=0
      boundBone {boneName=dyn_back_hem_l}
      boxExtents=<10 20 20>
      constraintSetup {angularLimitsMax=<1 -10 5>, angularLimitsMin=<-1 -20 -5>, angularTarget=<0 0 0>}
      localJointOffset=<1 0 -0.2>
     }]
   }
   Anim Dynamics - Bone: dyn_front_hem_l(ComponentPose=@n3, Alpha=1, ExternalForce="0,0,0") {
    angularDampingOverride=0.9999
    angularSpringConstant=75
    bAngularSpring=true
    bGravityOverrideInSimSpace=true
    bOverrideAngularDamping=true
    simulationSpace=RootRelative
    alphaBoolBlend {blendOption=Linear}
    boundBone {boneName=dyn_front_hem_l}
    componentLinearAccScale=<1 1 1>
    componentLinearVelScale=<1 1 1>
    physicsBodyDefinitions [{
      sphereCollisionRadius=0
      boundBone {boneName=dyn_front_hem_l}
      boxExtents=<10 20 20>
      constraintSetup {angularLimitsMax=<1 -2 0>, angularLimitsMin=<-1 -20 -10>, angularTarget=<0 0 0>}
      localJointOffset=<1 0 -0.2>
     }]
   }
   Control Rig() {
    alphaBoolBlend {blendOption=Linear}
    controlRigAssetReference {blueprintRigClass=&/M_MED_Jumpsuit_CtrlRig.M_MED_Jumpsuit_CtrlRig_C}
   }
   n4 = Anim Dynamics - Bone: dyn_front_hem_l(ComponentPose=Local To Component(LocalPose=Use cached pose 'WPO'()), Alpha=1, ExternalForce="0,0,0") {
    angularDampingOverride=0.9999
    angularSpringConstant=75
    bAngularSpring=true
    bGravityOverrideInSimSpace=true
    bOverrideAngularDamping=true
    simulationSpace=RootRelative
    alphaBoolBlend {blendOption=Linear}
    boundBone {boneName=dyn_front_hem_l}
    componentLinearAccScale=<1 1 1>
    componentLinearVelScale=<1 1 1>
    physicsBodyDefinitions [{
      sphereCollisionRadius=0
      boundBone {boneName=dyn_front_hem_l}
      boxExtents=<10 20 20>
      constraintSetup {angularLimitsMax=<1 -5 0>, angularLimitsMin=<-1 -45 -10>, angularTarget=<0 0 0>}
      localJointOffset=<1 0 -0.2>
     }]
   }
   n5 = Anim Dynamics - Bone: dyn_back_hem_l(ComponentPose=@n4, Alpha=1, ExternalForce="0,0,0") {
    angularDampingOverride=0.9999
    angularSpringConstant=75
    bAngularSpring=true
    bGravityOverrideInSimSpace=true
    bOverrideAngularDamping=true
    simulationSpace=RootRelative
    alphaBoolBlend {blendOption=Linear}
    boundBone {boneName=dyn_back_hem_l}
    componentLinearAccScale=<1 1 1>
    componentLinearVelScale=<1 1 1>
    physicsBodyDefinitions [{
      sphereCollisionRadius=0
      boundBone {boneName=dyn_back_hem_l}
      boxExtents=<10 20 20>
      constraintSetup {angularLimitsMax=<1 -5 5>, angularLimitsMin=<-1 -45 -5>, angularTarget=<0 0 0>}
      localJointOffset=<1 0 -0.2>
     }]
   }
   n6 = Anim Dynamics - Bone: dyn_front_hem_r(ComponentPose=@n5, Alpha=1, ExternalForce="0,0,0") {
    angularDampingOverride=0.9999
    angularSpringConstant=75
    bAngularSpring=true
    bGravityOverrideInSimSpace=true
    bOverrideAngularDamping=true
    simulationSpace=RootRelative
    alphaBoolBlend {blendOption=Linear}
    boundBone {boneName=dyn_front_hem_r}
    componentLinearAccScale=<1 1 1>
    componentLinearVelScale=<1 1 1>
    physicsBodyDefinitions [{
      sphereCollisionRadius=0
      boundBone {boneName=dyn_front_hem_r}
      boxExtents=<10 20 20>
      constraintSetup {angularLimitsMax=<1 5 10>, angularLimitsMin=<-1 -10 0>, angularTarget=<0 0 0>}
      localJointOffset=<1 0 -0.2>
     }]
   }
   n7 = Anim Dynamics - Bone: dyn_back_hem_r(ComponentPose=@n6, Alpha=1, ExternalForce="0,0,0") {
    angularDampingOverride=0.9999
    angularSpringConstant=75
    bAngularSpring=true
    bGravityOverrideInSimSpace=true
    bOverrideAngularDamping=true
    simulationSpace=RootRelative
    alphaBoolBlend {blendOption=Linear}
    boundBone {boneName=dyn_back_hem_r}
    componentLinearAccScale=<1 1 1>
    componentLinearVelScale=<1 1 1>
    physicsBodyDefinitions [{
      sphereCollisionRadius=0
      boundBone {boneName=dyn_back_hem_r}
      boxExtents=<10 20 20>
      constraintSetup {angularLimitsMax=<1 5 5>, angularLimitsMin=<-1 -10 -5>, angularTarget=<0 0 0>}
      localJointOffset=<1 0 -0.2>
     }]
   }
   n8 = Anim Dynamics - Bone: dyn_front_hem_l(ComponentPose=Local To Component(LocalPose=Use cached pose 'WPO'()), Alpha=1, ExternalForce="0,0,0") {
    angularDampingOverride=0.9999
    angularSpringConstant=75
    bAngularSpring=true
    bGravityOverrideInSimSpace=true
    bOverrideAngularDamping=true
    simulationSpace=RootRelative
    alphaBoolBlend {blendOption=Linear}
    boundBone {boneName=dyn_front_hem_l}
    componentLinearAccScale=<1 1 1>
    componentLinearVelScale=<1 1 1>
    physicsBodyDefinitions [{
      sphereCollisionRadius=0
      boundBone {boneName=dyn_front_hem_l}
      boxExtents=<10 20 20>
      constraintSetup {angularLimitsMax=<0 0 0>, angularLimitsMin=<0 -30 0>, angularTarget=<0 0 0>}
      localJointOffset=<3 0 0>
     }]
   }
   n9 = Anim Dynamics - Bone: dyn_back_hem_l(ComponentPose=@n8, Alpha=1, ExternalForce="0,0,0") {
    angularDampingOverride=0.9999
    angularSpringConstant=75
    bAngularSpring=true
    bGravityOverrideInSimSpace=true
    bOverrideAngularDamping=true
    simulationSpace=RootRelative
    alphaBoolBlend {blendOption=Linear}
    boundBone {boneName=dyn_back_hem_l}
    componentLinearAccScale=<1 1 1>
    componentLinearVelScale=<1 1 1>
    physicsBodyDefinitions [{
      sphereCollisionRadius=0
      boundBone {boneName=dyn_back_hem_l}
      boxExtents=<10 20 20>
      constraintSetup {angularLimitsMax=<1 -2 5>, angularLimitsMin=<-1 -45 -5>, angularTarget=<0 0 0>}
      localJointOffset=<1 0 -0.2>
     }]
   }
   n10 = Anim Dynamics - Bone: dyn_front_hem_r(ComponentPose=@n9, Alpha=1, ExternalForce="0,0,0") {
    angularDampingOverride=0.9999
    angularSpringConstant=75
    bAngularSpring=true
    bGravityOverrideInSimSpace=true
    bOverrideAngularDamping=true
    simulationSpace=RootRelative
    alphaBoolBlend {blendOption=Linear}
    boundBone {boneName=dyn_front_hem_r}
    componentLinearAccScale=<1 1 1>
    componentLinearVelScale=<1 1 1>
    physicsBodyDefinitions [{
      sphereCollisionRadius=0
      boundBone {boneName=dyn_front_hem_r}
      boxExtents=<10 20 20>
      constraintSetup {angularLimitsMax=<0 0 3>, angularLimitsMin=<0 -10 -3>, angularTarget=<0 0 0>}
      localJointOffset=<-3 0 0>
     }]
   }
   n11 = Anim Dynamics - Bone: dyn_back_hem_r(ComponentPose=@n10, Alpha=1, ExternalForce="0,0,0") {
    angularDampingOverride=0.9999
    angularSpringConstant=75
    bAngularSpring=true
    bGravityOverrideInSimSpace=true
    bOverrideAngularDamping=true
    simulationSpace=RootRelative
    alphaBoolBlend {blendOption=Linear}
    boundBone {boneName=dyn_back_hem_r}
    componentLinearAccScale=<1 1 1>
    componentLinearVelScale=<1 1 1>
    physicsBodyDefinitions [{
      sphereCollisionRadius=0
      boundBone {boneName=dyn_back_hem_r}
      boxExtents=<10 20 20>
      constraintSetup {angularLimitsMax=<0 -1 5>, angularLimitsMin=<0 -30 -5>, angularTarget=<0 0 0>}
      localJointOffset=<-3 0 0>
     }]
   }
   n12 = "Transform (Modify) Bone - Bone: MSTR_front_hem_l"(ComponentPose=@n11, Alpha=1, Translation="0,0,0", Rotation="5,0,0", Scale="1,1,1") {
    rotationMode=BMM_Additive
    rotationSpace=BCS_BoneSpace
    translationSpace=BCS_BoneSpace
    alphaBoolBlend {blendOption=Linear}
    boneToModify {boneName=MSTR_front_hem_l}
    rotation {pitch=5}
   }
   n13 = "Transform (Modify) Bone - Bone: MSTR_front_hem_r"(ComponentPose=@n12, Alpha=1, Translation="0,0,0", Rotation="-2.5,-7.5,0", Scale="1,1,1") {
    rotationMode=BMM_Additive
    rotationSpace=BCS_BoneSpace
    translationSpace=BCS_BoneSpace
    alphaBoolBlend {blendOption=Linear}
    boneToModify {boneName=MSTR_front_hem_r}
    rotation {pitch=-2.5, yaw=-7.5}
   }
   n14 = "Transform (Modify) Bone - Bone: MSTR_back_hem_r"!(ComponentPose=@n13, bAlphaBoolEnabled:bool=Get bIsCrouching(), Translation="0,0,1", Rotation="10,0,-10", Scale="0.8,1,1") -> (Pose) {
    alphaInputType=Bool
    rotationMode=BMM_Additive
    rotationSpace=BCS_BoneSpace
    scaleMode=BMM_Additive
    scaleSpace=BCS_BoneSpace
    translationMode=BMM_Additive
    translationSpace=BCS_BoneSpace
    alphaBoolBlend {blendOption=Linear}
    boneToModify {boneName=MSTR_back_hem_r}
    rotation {pitch=10, roll=-10}
    scale {x=0.8}
    translation {z=1}
   }
   Output Pose(Result=Blend Poses by bool(BlendPose_0=Component To Local(ComponentPose=@n7), BlendPose_1=Component To Local(ComponentPose=@n14)) {blendType=Linear}) {layerGroup=None}
   n15 = RigidBody(ComponentPose=Local To Component(LocalPose=Copy Pose From Mesh() {bCopyCurves=true, bUseAttachedParent=true})) {
    bOverrideWorldGravity=true
    overridePhysicsAsset=&/M_MED_Jumpsuit_Physics
    simulationSpace=BaseBoneSpace
    alphaBoolBlend {blendOption=Linear}
    baseBoneRef {boneName=pelvis}
    componentAppliedLinearAccClamp=<3000 3000 3000>
    componentLinearAccScale=<0.5 0.25 0.25>
    componentLinearVelScale=<0.25 0.1 0.1>
    overrideWorldGravity {z=-980}
   }
   n16 = Control Rig(Source=Blend Poses by bool(
     BlendPose_0=Component To Local(ComponentPose=@n15)
     BlendPose_1=Copy Pose From Mesh() {bCopyCurves=true, bUseAttachedParent=true}
     bActiveValue=Get bIsInFrontEnd()
    ) {blendType=Linear}) {
    alphaBoolBlend {blendOption=Linear}
    controlRigAssetReference {blueprintRigClass=&M_MED_Jumpsuit_CtrlRig_C}
   }
   n17 = Control Rig(Source=@n16) {
    alphaBoolBlend {blendOption=Linear}
    controlRigAssetReference {blueprintRigClass=&/M_MED_Jumpsuit_DeformRig_Master_ControlRig.M_MED_Jumpsuit_DeformRig_Master_ControlRig_C}
   }
   n18 = "Constraint - MSTR_pouchFlap_hip (P)"!(ComponentPose=Local To Component(LocalPose=@n17), Alpha=1, ConstraintWeights_0=1, ConstraintWeights_1:float=0) -> (Pose) {
    alphaBoolBlend {blendOption=Linear}
    boneToModify {boneName=MSTR_pouchFlap_hip}
    constraintSetup [{targetBone {boneName=pelvis}}]
    constraintWeights [1,0]
   }
   n19 = Anim Dynamics - Bone: dyn_pouchFlap_hip_0(ComponentPose=@n18, Alpha=1, ExternalForce="0,0,0") {
    angularDampingOverride=0.99
    bOverrideAngularDamping=true
    bUseGravityOverride=true
    bUseSphericalLimits=true
    simulationSpace=RootRelative
    alphaBoolBlend {blendOption=Linear}
    boundBone {boneName=dyn_pouchFlap_hip_0}
    componentAppliedLinearAccClamp=<3000 3000 3000>
    componentLinearAccScale=<0.6 0.6 0.45>
    componentLinearVelScale=<0.3 0.3 0.2>
    gravityOverride {z=-980}
    physicsBodyDefinitions [{
      boundBone {boneName=dyn_pouchFlap_hip_0}
      boxExtents=<5 20 20>
      constraintSetup {angularLimitsMax=<5 0 10>, angularLimitsMin=<-5 -70 -10>, angularTarget=<0 0 0>}
      localJointOffset=<5 0 0>
     }]
    sphericalLimits [{limitRadius=13.5, drivingBone {boneName=thigh_r}, sphereLocalOffset=<5 2.2221 -2.5057>}]
   }
   n20 = Anim Dynamics - Bone: dyn_pouchFlap_hip_1(ComponentPose=@n19, Alpha=0.7, ExternalForce="0,0,0") {
    angularDampingOverride=0.99
    bOverrideAngularDamping=true
    bUseGravityOverride=true
    simulationSpace=RootRelative
    alphaBoolBlend {blendOption=Linear}
    boundBone {boneName=dyn_pouchFlap_hip_1}
    componentAppliedLinearAccClamp=<3000 3000 3000>
    componentLinearAccScale=<0.6 0.6 0.45>
    componentLinearVelScale=<0.3 0.3 0.2>
    gravityOverride {z=-980}
    physicsBodyDefinitions [{
      boundBone {boneName=dyn_pouchFlap_hip_1}
      boxExtents=<5 20 20>
      constraintSetup {angularLimitsMax=<5 5 10>, angularLimitsMin=<-5 -80 -10>, angularTarget=<0 0 0>}
      localJointOffset=<5 0 0>
     }]
   }
   n21 = Anim Dynamics - Bone: dyn_pouchFlap_hip_2(ComponentPose=@n20, Alpha=0.7, ExternalForce="0,0,0") {
    angularDampingOverride=0.99
    bOverrideAngularDamping=true
    bUseGravityOverride=true
    simulationSpace=RootRelative
    alphaBoolBlend {blendOption=Linear}
    boundBone {boneName=dyn_pouchFlap_hip_2}
    componentAppliedLinearAccClamp=<3000 3000 3000>
    componentLinearAccScale=<0.6 0.6 0.45>
    componentLinearVelScale=<0.3 0.3 0.2>
    gravityOverride {z=-980}
    physicsBodyDefinitions [{
      boundBone {boneName=dyn_pouchFlap_hip_2}
      boxExtents=<5 20 20>
      constraintSetup {angularLimitsMax=<5 5 10>, angularLimitsMin=<-5 -90 -10>, angularTarget=<0 0 0>}
      localJointOffset=<5 0 0>
     }]
   }
   n22 = Anim Dynamics - Bone: dyn_pouchFlap_0_r(ComponentPose=@n21, Alpha=1, ExternalForce="0,0,0") {
    angularDampingOverride=0.999
    bOverrideAngularDamping=true
    simulationSpace=RootRelative
    alphaBoolBlend {blendOption=Linear}
    boundBone {boneName=dyn_pouchFlap_0_r}
    componentAppliedLinearAccClamp=<3000 3000 3000>
    componentLinearAccScale=<0.5 0.5 0.5>
    componentLinearVelScale=<0.25 0.25 0.25>
    gravityOverride {z=-980}
    physicsBodyDefinitions [{
      boundBone {boneName=dyn_pouchFlap_0_r}
      boxExtents=<5 20 20>
      constraintSetup {angularLimitsMax=<5 0 2>, angularLimitsMin=<-5 -90 -2>, angularTarget=<0 0 0>}
      localJointOffset=<5 0 0>
     }]
   }
   n23 = Anim Dynamics - Bone: dyn_pouchFlap_1_r(ComponentPose=@n22, Alpha=1, ExternalForce="0,0,0") {
    angularDampingOverride=0.999
    bOverrideAngularDamping=true
    simulationSpace=RootRelative
    alphaBoolBlend {blendOption=Linear}
    boundBone {boneName=dyn_pouchFlap_1_r}
    componentAppliedLinearAccClamp=<3000 3000 3000>
    componentLinearAccScale=<0.5 0.5 0.5>
    componentLinearVelScale=<0.25 0.25 0.25>
    gravityOverride {z=-980}
    physicsBodyDefinitions [{
      boundBone {boneName=dyn_pouchFlap_1_r}
      boxExtents=<5 20 20>
      constraintSetup {angularLimitsMax=<5 5 2>, angularLimitsMin=<-5 -90 -2>, angularTarget=<0 0 0>}
      localJointOffset=<5 0 0>
     }]
   }
   n24 = Anim Dynamics - Bone: dyn_pouchFlap_2_r(ComponentPose=@n23, Alpha=1, ExternalForce="0,0,0") {
    angularDampingOverride=0.999
    bOverrideAngularDamping=true
    simulationSpace=RootRelative
    alphaBoolBlend {blendOption=Linear}
    boundBone {boneName=dyn_pouchFlap_2_r}
    componentAppliedLinearAccClamp=<3000 3000 3000>
    componentLinearAccScale=<0.5 0.5 0.5>
    componentLinearVelScale=<0.25 0.25 0.25>
    gravityOverride {z=-980}
    physicsBodyDefinitions [{
      boundBone {boneName=dyn_pouchFlap_2_r}
      boxExtents=<5 20 20>
      constraintSetup {angularLimitsMax=<5 5 2>, angularLimitsMin=<-5 -90 -2>, angularTarget=<0 0 0>}
      localJointOffset=<5 0 0>
     }]
   }
   n25 = Anim Dynamics - Bone: dyn_pouchFlap_0_l(ComponentPose=@n24, Alpha=1, ExternalForce="0,0,0") {
    angularDampingOverride=0.9999
    bOverrideAngularDamping=true
    bUsePlanarLimit=false
    bUseSphericalLimits=true
    simulationSpace=RootRelative
    alphaBoolBlend {blendOption=Linear}
    boundBone {boneName=dyn_pouchFlap_0_l}
    componentAppliedLinearAccClamp=<5000 5000 5000>
    componentLinearAccScale=<0.5 0.5 0.5>
    componentLinearVelScale=<0.25 0.25 0.25>
    gravityOverride {z=-980}
    physicsBodyDefinitions [{
      boundBone {boneName=dyn_pouchFlap_0_l}
      boxExtents=<5 20 20>
      constraintSetup {angularLimitsMax=<5 20 5>, angularLimitsMin=<-5 -90 -90>, angularTarget=<0 0 0>}
      localJointOffset=<5 0 0>
     }]
    sphericalLimits [
     {limitRadius=7, drivingBone {boneName=calf_l}, sphereLocalOffset=<6 -2 -1.5>}
     {limitRadius=0, drivingBone {boneName=None}, sphereLocalOffset=<0 0 0>}
    ]
   }
   n26 = Anim Dynamics - Bone: dyn_pouchFlap_1_l(ComponentPose=@n25, Alpha=1, ExternalForce="0,0,0") {
    angularDampingOverride=0.9999
    bOverrideAngularDamping=true
    bUseSphericalLimits=true
    simulationSpace=RootRelative
    alphaBoolBlend {blendOption=Linear}
    boundBone {boneName=dyn_pouchFlap_1_l}
    componentAppliedLinearAccClamp=<5000 5000 5000>
    componentLinearAccScale=<0.5 0.5 0.5>
    componentLinearVelScale=<0.25 0.25 0.25>
    gravityOverride {z=-980}
    physicsBodyDefinitions [{
      boundBone {boneName=dyn_pouchFlap_1_l}
      boxExtents=<5 20 20>
      constraintSetup {angularLimitsMax=<5 20 5>, angularLimitsMin=<-5 -90 -5>, angularTarget=<0 0 0>}
      localJointOffset=<5 0 0>
     }]
    sphericalLimits [
     {limitRadius=7.5, drivingBone {boneName=calf_l}, sphereLocalOffset=<5 -1.2 -1>}
     {limitRadius=7, drivingBone {boneName=calf_l}, sphereLocalOffset=<-2.5 -7 -5>}
    ]
   }
   n27 = Anim Dynamics - Bone: dyn_pouchFlap_2_l(ComponentPose=@n26, Alpha=1, ExternalForce="0,0,0") {
    angularDampingOverride=0.9999
    bOverrideAngularDamping=true
    simulationSpace=RootRelative
    alphaBoolBlend {blendOption=Linear}
    boundBone {boneName=dyn_pouchFlap_2_l}
    componentAppliedLinearAccClamp=<5000 5000 5000>
    componentLinearAccScale=<0.5 0.5 0.5>
    componentLinearVelScale=<0.25 0.25 0.25>
    gravityOverride {z=-980}
    physicsBodyDefinitions [{
      boundBone {boneName=dyn_pouchFlap_2_l}
      boxExtents=<5 20 20>
      constraintSetup {angularLimitsMax=<5 20 5>, angularLimitsMin=<-5 -90 -5>, angularTarget=<0 0 0>}
      localJointOffset=<5 0 0>
     }]
   }
   n28 = Anim Dynamics - Bone: dyn_chest_zipper(ComponentPose=@n27, Alpha=1, ExternalForce="0,0,0") {
    angularDampingOverride=0.9999
    bOverrideAngularDamping=true
    bUseGravityOverride=true
    bUsePlanarLimit=false
    simulationSpace=RootRelative
    alphaBoolBlend {blendOption=Linear}
    boundBone {boneName=dyn_chest_zipper}
    componentAppliedLinearAccClamp=<3000 3000 3000>
    componentLinearAccScale=<0.5 0.5 0.5>
    componentLinearVelScale=<0.25 0.25 0.25>
    gravityOverride {z=-980}
    physicsBodyDefinitions [{
      boundBone {boneName=dyn_chest_zipper}
      boxExtents=<2 10 10>
      constraintSetup {angularLimitsMax=<2 5 30>, angularLimitsMin=<-2 -30 -30>, angularTarget=<0 0 0>}
      localJointOffset=<3 0 0>
     }]
   }
   n29 = "Constraint - MSTR_chest_zipper (PPPPP)"!(
    ComponentPose=@n28
    Alpha=1
    ConstraintWeights_0=0.0031
    ConstraintWeights_1:float=0.1636
    ConstraintWeights_2:float=0.4386
    ConstraintWeights_3:float=0.0844
    ConstraintWeights_4:float=0.3103
   ) -> (Pose) {
    alphaBoolBlend {blendOption=Linear}
    boneToModify {boneName=MSTR_chest_zipper}
    constraintSetup [
     {targetBone {boneName=spine_01}}
     {targetBone {boneName=spine_02}}
     {targetBone {boneName=spine_04}}
     {targetBone {boneName=spine_05}}
     {targetBone {boneName=spine_03}}
    ]
    constraintWeights [1,1,1,1,1]
   }
   n30 = Anim Dynamics - Bone: dyn_pouch_zipper(ComponentPose=@n29, Alpha=1, ExternalForce="0,0,0") {
    angularDampingOverride=0.9999
    bOverrideAngularDamping=true
    bUseGravityOverride=true
    bUsePlanarLimit=false
    simulationSpace=RootRelative
    alphaBoolBlend {blendOption=Linear}
    boundBone {boneName=dyn_pouch_zipper}
    componentAppliedLinearAccClamp=<3000 3000 3000>
    componentLinearAccScale=<0.5 0.5 0.5>
    componentLinearVelScale=<0.25 0.25 0.25>
    gravityOverride {z=-980}
    physicsBodyDefinitions [{
      boundBone {boneName=dyn_pouch_zipper}
      boxExtents=<2 10 10>
      constraintSetup {angularLimitsMax=<2 5 30>, angularLimitsMin=<-2 -30 -30>, angularTarget=<0 0 0>}
      localJointOffset=<3 0 0>
     }]
   }
   n31 = "Constraint - MSTR_pouch_zipper (PPP)"!(ComponentPose=@n30, Alpha=1, ConstraintWeights_0=0.008, ConstraintWeights_1:float=0.0042, ConstraintWeights_2:float=0.9878) -> (Pose) {
    alphaBoolBlend {blendOption=Linear}
    boneToModify {boneName=MSTR_pouch_zipper}
    constraintSetup [{targetBone {boneName=thigh_l}}, {targetBone {boneName=pelvis}}, {targetBone {boneName=thigh_twist_01_l}}]
    constraintWeights [1,1,1]
   }
   n32 = Anim Dynamics - Bone: dyn_leg_l_zipper(ComponentPose=@n31, Alpha=1, ExternalForce="0,0,0") {
    angularDampingOverride=0.9999
    bOverrideAngularDamping=true
    bUseGravityOverride=true
    bUsePlanarLimit=false
    simulationSpace=RootRelative
    alphaBoolBlend {blendOption=Linear}
    boundBone {boneName=dyn_leg_l_zipper}
    componentAppliedLinearAccClamp=<3000 3000 3000>
    componentLinearAccScale=<0.5 0.5 0.5>
    componentLinearVelScale=<0.25 0.25 0.25>
    gravityOverride {z=-980}
    physicsBodyDefinitions [{
      boundBone {boneName=dyn_leg_l_zipper}
      boxExtents=<2 10 10>
      constraintSetup {angularLimitsMax=<2 5 60>, angularLimitsMin=<-2 -30 -90>, angularTarget=<0 0 0>}
      localJointOffset=<3 0 0>
     }]
   }
   n33 = "Constraint - MSTR_leg_l_zipper (PPPPP)"!(
    ComponentPose=@n32
    Alpha=1
    ConstraintWeights_0=0.0724
    ConstraintWeights_1:float=0.1378
    ConstraintWeights_2:float=0.0015
    ConstraintWeights_3:float=0.0001
    ConstraintWeights_4:float=0.7882
   ) -> (Pose) {
    alphaBoolBlend {blendOption=Linear}
    boneToModify {boneName=MSTR_leg_l_zipper}
    constraintSetup [
     {targetBone {boneName=pelvis}}
     {targetBone {boneName=thigh_l}}
     {targetBone {boneName=calf_l}}
     {targetBone {boneName=calf_twist_02_l}}
     {targetBone {boneName=thigh_twist_01_l}}
    ]
    constraintWeights [1,1,1,1,1]
   }
   n34 = Anim Dynamics - Bone: dyn_leg_r_zipper(ComponentPose=@n33, Alpha=1, ExternalForce="0,0,0") {
    angularDampingOverride=0.9999
    bOverrideAngularDamping=true
    bUseGravityOverride=true
    bUsePlanarLimit=false
    simulationSpace=RootRelative
    alphaBoolBlend {blendOption=Linear}
    boundBone {boneName=dyn_leg_r_zipper}
    componentAppliedLinearAccClamp=<3000 3000 3000>
    componentLinearAccScale=<0.5 0.5 0.5>
    componentLinearVelScale=<0.25 0.25 0.25>
    gravityOverride {z=-980}
    physicsBodyDefinitions [{
      boundBone {boneName=dyn_leg_r_zipper}
      boxExtents=<2 10 10>
      constraintSetup {angularLimitsMax=<2 5 30>, angularLimitsMin=<-2 -30 -30>, angularTarget=<0 0 0>}
      localJointOffset=<3 0 0>
     }]
   }
   n35 = "Constraint - MSTR_leg_r_zipper (PPPPPPPP)"!(
    ComponentPose=@n34
    Alpha=1
    ConstraintWeights_0=0.2686
    ConstraintWeights_1:float=0.0012
    ConstraintWeights_2:float=0
    ConstraintWeights_3:float=0.0035
    ConstraintWeights_4:float=0.0998
    ConstraintWeights_5:float=0.2228
    ConstraintWeights_6:float=0.0002
    ConstraintWeights_7:float=0.4038
   ) -> (Pose) {
    alphaBoolBlend {blendOption=Linear}
    boneToModify {boneName=MSTR_leg_r_zipper}
    constraintSetup [
     {targetBone {boneName=pelvis}}
     {targetBone {boneName=spine_01}}
     {targetBone {boneName=spine_02}}
     {targetBone {boneName=thigh_l}}
     {targetBone {boneName=thigh_twist_01_l}}
     {targetBone {boneName=thigh_r}}
     {targetBone {boneName=calf_r}}
     {targetBone {boneName=thigh_twist_01_r}}
    ]
    constraintWeights [1,1,1,1,1,1,1,1]
   }
   n36 = "Constraint - MSTR_front_hem_l (R)"(ComponentPose=@n35, Alpha=1, ConstraintWeights_0=1) {
    alphaBoolBlend {blendOption=Linear}
    boneToModify {boneName=MSTR_front_hem_l}
    constraintSetup [{transformType=Rotation, targetBone {boneName=foot_l}}]
    constraintWeights [1]
   }
   n37 = "Constraint - MSTR_front_hem_r (R)"(ComponentPose=@n36, Alpha=Get NotCrouching() -> (NotCrouching:double), ConstraintWeights_0=1) {
    alphaBoolBlend {blendOption=Linear}
    boneToModify {boneName=MSTR_front_hem_r}
    constraintSetup [{transformType=Rotation, targetBone {boneName=foot_r}}]
    constraintWeights [1]
   }
   Save cached pose 'WPO'(Pose=Component To Local(ComponentPose=@n37)) {cacheName=WPO}
  ]}
 EventGraph {nodes [
   n1 = Event BlueprintUpdateAnimation()
   > n2 = Sequence!(execute) -> (then_0, then_1, then_2:Exec)
   > Set WPO Globals Function(
    Material=Get Mat Instance Dynamic()
    Skeletal Mesh=Get Skel Mesh()
    "Local Noise Transform  Location"=Get Local Noise Transform Location() -> (Local Noise Transform Location:Vector)
    Local Noise Offset=Get Local Noise Offset()
   )
   > Set WPO Regional Function!(
    execute
    Material:Material Instance Dynamic Object Reference=Get Mat Instance Dynamic()
    Overlay Material:Material Instance Dynamic Object Reference
    Skeletal Mesh:Skeletal Mesh Component Object Reference=Get Skel Mesh()
    Velocity Transform:Name=foot_r
    Velocity Interp Speed:double=Get Velocity Interp Speed()
    Collision Transform:Name=calf_r
    Collision Transform Axis:EAxis Enum
    Invert Collision Transform Axis:bool=false
    Delta Time X:double=@n1.DeltaTimeX
    Previous Velocity Transform Location:"Vector (by ref)"=Get R1 Velocity Transform Location()
    Previous Velocity Transform Velocity:"Vector (by ref)"=Get R1 Velocity Transform Velocity()
    Region:int=1
   ) -> (then)
   > Set WPO Regional Function(
    Material=Get Mat Instance Dynamic()
    Skeletal Mesh=Get Skel Mesh()
    Velocity Transform=foot_l
    Velocity Interp Speed=Get Velocity Interp Speed()
    Collision Transform=calf_l
    Invert Collision Transform Axis=true
    Delta Time X=@n1.DeltaTimeX
    Previous Velocity Transform Location=Get R2 Velocity Transform Location()
    Previous Velocity Transform Velocity=Get R2 Velocity Transform Velocity()
    Region=2
   )
   > n3 = Branch(Condition=Get bIsSkydiveFloating())
   > n4 = Set WPO Pho Wind Intensity(WPO Pho Wind Intensity=1)
   SetScalarParameterValue(execute=[@n5.then, @n4.then], self=Get Mat Instance Dynamic(), ParameterName=ClothSim_WindStrength, Value=Get WPO Pho Wind Intensity())
   n5 = Set WPO Pho Wind Intensity(execute=@n3.else, WPO Pho Wind Intensity=0)
   n6 = Branch(execute=@n2.then_1, Condition=NOT Boolean(A=Get bIsCrouching()))
   > Set NotCrouching(NotCrouching:double=1) -> (Output_Get:double)
   Set NotCrouching(execute=@n6.else, NotCrouching=0)
   Event BlueprintInitializeAnimation()
   > Set Skel Mesh(Skel Mesh=GetOwnerSkeletalMeshComponent())
   n7 = TryGetPawnOwner()
   > Is Valid(InputObject=@n7)
   > n8 = Cast To FortPlayerPawnAthena(Object=@n7)
   > Set Player Pawn(Player Pawn:Fort Player Pawn Athena Object Reference=@n8) -> (Output_Get:Fort Player Pawn Athena Object Reference)
   Event OnCharacterCustomizationLoaded()
   > n9 = Set Skel Mesh(Skel Mesh=GetPartSkeletalMeshComponent())
   > n10 = Get Dynamic Material Instance(SkeletalMeshComponent=@n9)
   > Set Mat Instance Dynamic(Mat Instance Dynamic=@n10)
  ]}
]
variables [
 Local Noise Offset {type=Vector}
 Local Noise Transform Location {type=Vector}
 Mat Instance Dynamic {type=MaterialInstanceDynamic}
 NotCrouching {type=number, default=1}
 Player Pawn {type=FortPlayerPawnAthena}
 R1 Velocity Transform Location {type=Vector}
 R1 Velocity Transform Velocity {type=Vector}
 R2 Velocity Transform Location {type=Vector}
 R2 Velocity Transform Velocity {type=Vector}
 Skel Mesh {type=SkeletalMeshComponent}
 Velocity Interp Speed {type=number, default=30}
 WPO Pho Wind Intensity {type=number}
 WPO Pho Wind Intensity_0 {type=number, default=1}
]
