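# Animation Blueprint for the male medium jumpsuit's face accessory skeleton, adding physical jiggle to ear and dog-ear bones.
# The incoming Input Pose runs through four chained Anim Dynamics nodes, one per left or right bone, each at half alpha.
# Each simulates root-relative with overridden gravity and damping, scaled component motion and narrow angular limits.
# A Copy Pose From Mesh node sits unconnected, and the event graph has no logic.
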AnimBlueprint /Game/Characters/Player/Male/Medium/Bodies/M_MED_Jumpsuit/Meshes/Parts/M_MED_Jumpsuit_FaceAcc_AnimBP
target_skeleton=&/M_MED_Jumpsuit_FaceAcc_Skeleton
graphs [
 AnimGraph {nodes [
   Copy Pose From Mesh() {bUseAttachedParent=true}
   n1 = Anim Dynamics - Bone: dyn_ear_r(ComponentPose=Local To Component(LocalPose=Input Pose()), Alpha=0.5, ExternalForce="0,0,0") {
    angularDampingOverride=0.99
    bOverrideAngularDamping=true
    bUseGravityOverride=true
    simulationSpace=RootRelative
    alphaBoolBlend {blendOption=Linear}
    boundBone {boneName=dyn_ear_r}
    componentLinearAccScale=<0.5 0.2 0.2>
    componentLinearVelScale=<0.25 0.1 0.1>
    gravityOverride {z=-980}
    physicsBodyDefinitions [{
      boundBone {boneName=dyn_ear_r}
      boxExtents=<2 10 10>
      constraintSetup {angularLimitsMax=<0 10 10>, angularLimitsMin=<0 -10 -10>, angularTarget=<0 0 0>}
      localJointOffset=<-3 0 0>
     }]
   }
   n2 = Anim Dynamics - Bone: dyn_ear_l(ComponentPose=@n1, Alpha=0.5, ExternalForce="0,0,0") {
    angularDampingOverride=0.9999
    bOverrideAngularDamping=true
    bUseGravityOverride=true
    simulationSpace=RootRelative
    alphaBoolBlend {blendOption=Linear}
    boundBone {boneName=dyn_ear_l}
    componentLinearAccScale=<0.5 0.2 0.2>
    componentLinearVelScale=<0.25 0.1 0.1>
    gravityOverride {z=-980}
    physicsBodyDefinitions [{
      boundBone {boneName=dyn_ear_l}
      boxExtents=<2 10 10>
      constraintSetup {angularLimitsMax=<0 10 10>, angularLimitsMin=<0 -10 -10>, angularTarget=<0 0 0>}
      localJointOffset=<3 0 0>
     }]
   }
   n3 = Anim Dynamics - Bone: dyn_dog_ear_r(ComponentPose=@n2, Alpha=0.5, ExternalForce="0,0,0") {
    angularDampingOverride=0.9999
    bOverrideAngularDamping=true
    bUseGravityOverride=true
    simulationSpace=RootRelative
    alphaBoolBlend {blendOption=Linear}
    boundBone {boneName=dyn_dog_ear_r}
    componentLinearAccScale=<0.5 0.2 0.2>
    componentLinearVelScale=<0.25 0.1 0.1>
    gravityOverride {z=-980}
    physicsBodyDefinitions [{
      boundBone {boneName=dyn_dog_ear_r}
      boxExtents=<2 10 10>
      constraintSetup {angularLimitsMax=<0 10 10>, angularLimitsMin=<0 -10 -10>, angularTarget=<0 0 0>}
      localJointOffset=<-3 0 0>
     }]
   }
   n4 = Anim Dynamics - Bone: dyn_dog_ear_l(ComponentPose=@n3, Alpha=0.5, ExternalForce="0,0,0") {
    angularDampingOverride=0.99
    bOverrideAngularDamping=true
    bUseGravityOverride=true
    simulationSpace=RootRelative
    alphaBoolBlend {blendOption=Linear}
    boundBone {boneName=dyn_dog_ear_l}
    componentLinearAccScale=<0.5 0.2 0.2>
    componentLinearVelScale=<0.25 0.1 0.1>
    gravityOverride {z=-980}
    physicsBodyDefinitions [{
      boundBone {boneName=dyn_dog_ear_l}
      boxExtents=<2 10 10>
      constraintSetup {angularLimitsMax=<0 10 10>, angularLimitsMin=<0 -10 -10>, angularTarget=<0 0 0>}
      localJointOffset=<3 0 0>
     }]
   }
   Output Pose(Result=Component To Local(ComponentPose=@n4)) {layerGroup=None}
  ]}
 EventGraph {nodes [Event BlueprintUpdateAnimation()]}
]
variables []
